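; Athena beach ball prop Blueprint that records pawns touching it and reports kicks for quest progress.
; On actor begin overlap it casts the other actor to a Fort pawn, prints a debug string and stores that pawn as LastToKick.
; On the server only, it gets that player's Athena quest manager and sends a custom stat event tagged Athena.Quests.KickBeachball.
; The construction script is empty, and no physics impulse logic appears in the graph.
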(event Collision|EventActorBeginOverlap
  (params (OtherActor :type Actor%20Object%20Reference))
  (bind _asfort_pawn (Utilities|Casting|CastToFortPawn OtherActor))
  (Development|PrintString "KICK! Overlap" false false "(R=0.202252,G=1.000000,B=0.000000,A=1.000000)")
  (Variables|Default|SetLastToKick _asfort_pawn)
  ;[1 1 1] Quest Quest stat event
  (if (Networking|IsServer)
    (Quest|SendComplexCustomStatEvent (Quests|GetQuestManager (Fortnite|GetFortPlayerControllerfromActor _asfort_pawn) "Athena") 0 "(GameplayTags=((TagName=\"Athena.Quests.KickBeachball\")))" "(GameplayTags=)" 1)))

(fn ConstructionScript)
