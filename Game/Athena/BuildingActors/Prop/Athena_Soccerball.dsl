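; Athena soccer ball prop that reports how far a player kicks it for the KickSoccerball quest.
; On actor overlap it records its location and the kicking pawn; outside Creative, a valid, incomplete quest objective caches the quest manager and starts a one-second timer.
; The timer event converts 2D travel to meters, sends it as a custom quest stat if outside the 0–2 range, then clears the looping timer, so it acts as one-shot.
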(event Collision|EventActorBeginOverlap
  (params (OtherActor :type Actor%20Object%20Reference))
  (bind _self self)
  (Utilities|IsValid OtherActor
    (:Is%20Valid
      (Utilities|IsValid _self
        (:Is%20Valid
          (Variables|Default|Setoriginal%20location (Transformation|GetActorLocation _self))
          (bind _asfort_pawn (Utilities|Casting|CastToFortPawn OtherActor))
          (bind _output_get (Variables|Default|SetLastToKick _asfort_pawn))
          (if (Utilities|IsValid (Fortnite|GetFortPlayerControllerfromActor _output_get) :!class /Script/Engine.KismetSystemLibrary)
            (if (not (Class|FortGameStateAthena|IsCreativeModePlaylist (Fortnite|Athena|GameState|GetGameStateAthena)))
              (bind (_questisvalid _questcompleted _objectivecompleted) (Default|AthenaCheckQuestandObjectiveCompletion (Fortnite|GetFortPlayerControllerfromActor _output_get) (Variables|Default|GetQuest%20Item) (Variables|Default|GetQuest%20Backend%20Name)))
              (if (and _questisvalid (not _objectivecompleted))
                (Variables|Default|SetLast%20Kicker%20Quest%20Manager (Quests|GetQuestManager (Fortnite|GetFortPlayerControllerfromActor _output_get) "Invalid"))
                (bind _returnvalue (Utilities|Time|SetTimerbyEvent (delegate Custom|Update%20Quest%20With%20Distance) 1.0 true))
                (Variables|Default|SetTimerHandle _returnvalue)))))))))

(event Custom|Update%20Quest%20With%20Distance
  (bind _self self)
  (bind _returnvalue (Transformation|GetActorLocation _self))
  (bind _original_location (Variables|Default|Getoriginal%20location))
  (bind _returnvalue_2 (Math|Vector|Distance2D(Vector) _original_location _returnvalue))
  (bind _returnvalue_3 (Math|Float|Absolute(Float) _returnvalue_2))
  (bind _returnvalue_4 (/ _returnvalue_3 1e+02))
  (Utilities|IsValid _self
    (:Is%20Valid
      (if (not (Math|Float|InRange(Float) _returnvalue_4 0.0 2.0))
        (Quest|SendComplexCustomStatEvent (Variables|Default|GetLast%20Kicker%20Quest%20Manager) 0 "(GameplayTags=((TagName=\"Athena.Quests.KickSoccerball\")))" "(GameplayTags=)" (Math|Float|Truncate _returnvalue_4))
        (Variables|Default|Setoriginal%20location _returnvalue)))
    (:Is%20Not%20Valid
      _))
  (Utilities|Time|ClearandInvalidateTimerbyHandle (Variables|Default|GetTimerHandle)))

(fn ConstructionScript)
